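# Translucent, two-sided unlit Niagara material for stylized fire in the DuckCoast emote, usable on sprites, ribbons and meshes.
# A procedural tapered UV mask combines with panning wispy noise, colored through a toon curve atlas, hue shift and particle color.
# Sine-based WPO adds camera-relative wiggle.
# Opacity layers dynamic-parameter erosion, view-angle and optional axis fades, edge hardening, and depth fade, which Forward and Mobile skip.
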Material /BRCosmetics/Animation/Game/MainPlayer/Emotes/DuckCoast/FX/M_Emote_DuckCoast_Fire
graph {
 properties {
  bUsedWithNiagaraMeshParticles=true
  bUsedWithNiagaraRibbons=true
  bUsedWithNiagaraSprites=true
  blendMode=BLEND_Translucent
  pixelDepthOffsetMode=PDOM_Legacy
  shadingModel=MSM_Unlit
  twoSided=true
  usedShadingModels=Unlit
 }
 nodes [
  n1 = TextureCoordinate()
  n2 = StaticSwitchParameter(True=ComponentMask(None=@n1) {g=true}, False=OneMinus(None=ComponentMask(None=@n1) {r=true})) {parameterName=Swap UV}
  n3 = TextureCoordinate()
  n4 = ComponentMask(None=@n3) {r=true}
  n5 = ComponentMask(None=@n3) {g=true}
  n6 = Power(
   Base=SphereMask(
    A=StaticSwitchParameter(True=@n4, False=@n5) {parameterName=Swap UV}
    B=Constant() {r=0.5}
    Radius=ScalarParameter() {defaultValue=0.5, parameterName=Width, sliderMax=1}
   ) {attenuationRadius=0.5, hardnessPercent=0}
   Exp=ScalarParameter() {defaultValue=1, parameterName=Concave, sliderMax=10}
  )
  n7 = SmoothStep(
   Min=ScalarParameter() {parameterName=Pointiness, sliderMax=1}
   Value=Subtract(A=Multiply(A=@n2, B=@n6), B=Power(Base=@n2) {constExponent=30})
  )
  n8 = Multiply(A=@n7) {constB=3}
  n9 = Add(
   A=TextureCoordinate() {coordinateIndex=1}
   B=DynamicParameter() {defaultValue=<rgba 0 0.75 0 1>, paramNames [Offset, Local Axis Fade Min, AlphaErosion, Param4]}.Offset
  )
  n10 = AppendVector(
   A=ScalarParameter() {defaultValue=-1, group=Wispy Flame, parameterName=Speed X, sliderMax=5, sliderMin=-5}
   B=ScalarParameter() {group=Wispy Flame, parameterName=Speed Y, sliderMax=5, sliderMin=-5}
  )
  n11 = ScalarParameter() {defaultValue=1, group=Wispy Flame, parameterName=Scale X, sliderMax=5}
  n12 = ScalarParameter() {defaultValue=1, group=Wispy Flame, parameterName=Scale Y, sliderMax=5}
  n13 = TextureSampleParameter2D(UVs=Multiply(A=Panner(Coordinate=@n9, Speed=@n10) {speedX=1}, B=AppendVector(A=@n11, B=@n12))) {
   group=Wispy Flame
   parameterName=Flame Texture
   texture=/BRCosmetics/Athena/Effects/Particles/Skydiving/NS_Contrail_Chainmail/T_WispyFireNoise03
  }
  n14 = Multiply(A=Power(Base=@n2) {constExponent=1}, B=LinearInterpolate(A=Multiply(A=@n8, B=@n13.R), B=@n7, Alpha=@n7))
  n15 = Saturate(None=@n14)
  n16 = ScalarParameter() {defaultValue=1, group=Color, parameterName=Color Power, sliderMax=5}
  n17 = CurveAtlasRowParameter(CurveTime=SmoothStep(
    Min=ScalarParameter() {group=Color, parameterName=Color Smooth Step Min, sliderMax=1, sliderMin=-1}
    Max=ScalarParameter() {defaultValue=2, group=Color, parameterName=Color Smooth Step Max, sliderMax=5}
    Value=Power(Base=@n15, Exp=@n16) {constExponent=1}
   ) {constMax=2}) {atlas=/Game/Athena/LinearColorCurves/VFX_CurveAtlas_Toon, curve=/Game/Athena/LinearColorCurves/Curve_FireToon02, parameterName=Curve_Color}
  n18 = MaterialFunctionCall(Hue Shift Percentage=ScalarParameter() {parameterName=HueShift}, Texture=@n17) {materialFunction=&HueShift}
  n19 = ScalarParameter() {parameterName=Desaturation}
  n20 = ParticleColor()
  MaterialFunctionCall(Emissive=Multiply(A=Desaturation(None=@n18.Result, Fraction=@n19), B=@n20.RGB)) {materialFunction=&MF_FX_EmissiveAdjustments}
  n21 = Multiply(A=Time(), B=ScalarParameter() {defaultValue=1, group=WPO Wiggle, parameterName=Wiggle Speed, sliderMax=5, sliderMin=-5})
  n22 = ComponentMask(None=TextureCoordinate() {coordinateIndex=1}) {g=true}
  n23 = Sine(None=Multiply(A=Add(A=@n21, B=@n22), B=ScalarParameter() {defaultValue=2, group=WPO Wiggle, parameterName=Wiggle Frequency, sliderMax=10}) {constB=2})
  n24 = Power(
   Base=SmoothStep(Value=OneMinus(None=@n22)) {constMin=0.1}
   Exp=ScalarParameter() {defaultValue=1, group=WPO Wiggle, parameterName=Wiggle Falloff, sliderMax=5}
  )
  n25 = Constant3Vector() {constant=<rgba 0 1 0 0>}
  n26 = Normalize(VectorInput=Multiply(
    A=Transform(None=@n25) {transformSourceType=TRANSFORMSOURCE_Local, transformType=TRANSFORM_Camera}
    B=Constant3Vector() {constant=<rgba 1 1 0 0>}
   ))
  n27 = Multiply(
   A=Transform(None=Constant3Vector() {constant=<rgba -1 0 0 0>}) {transformSourceType=TRANSFORMSOURCE_Camera}
   B=DotProduct(A=@n26, B=Constant3Vector() {constant=<rgba 0 1 0 0>})
  )
  n28 = Multiply(
   A=Transform(None=Constant3Vector() {constant=<rgba 0 1 0 0>}) {transformSourceType=TRANSFORMSOURCE_Camera}
   B=DotProduct(A=@n26, B=Constant3Vector() {constant=<rgba 1 0 0 0>})
  )
  Multiply(A=Multiply(A=Multiply(A=@n23, B=@n24), B=Add(A=@n27, B=@n28)), B=ScalarParameter() {group=WPO Wiggle, parameterName=WPO Wiggle Wave})
  n29 = Multiply(A=@n14, B=@n20.A)
  n30 = DynamicParameter() {defaultValue=<rgba 0 0.75 0 1>, paramNames [Offset, Local Axis Fade Min, AlphaErosion, Param4]}
  Multiply(A=MaterialFunctionCall!(Texture=@n29, Erosion=@n30.AlphaErosion, Softness) {materialFunction=/Game/Effects/Niagara/MasterMaterials/Functions/MF_Erosion_Smoothstep_01}.Result)
  ScalarParameter() {parameterName=Fade Min Local, sliderMax=1}
  n31 = Subtract(A=@n30.AlphaErosion) {constB=0}
  n32 = MaterialFunctionCall(Input=@n30.AlphaErosion, Target Low=@n31) {materialFunction=&RemapValueRange}
  n33 = Add(A=@n32.Result, B=@n31)
  n34 = Multiply(
   A=@n29
   B=Saturate(None=MaterialFunctionCall(Alpha=@n29, Min=@n32.Result, Max=@n33) {materialFunction=&SmoothStep}.Result)
  )
  n35 = MaterialFunctionCall("FadeSensitivity (optional)"=ScalarParameter() {defaultValue=4, parameterName=ViewAngleFadeSensitivity, sliderMax=10}) {materialFunction=&FadeBasedOnViewAngle}
  n36 = DynamicParameter() {parameterIndex=1, defaultValue=<rgba 0 0 1 1>, paramNames [Axis X, Axis Y, Axis Z, Param4]}
  n37 = MaterialFunctionCall(R=@n36.Axis X, G=@n36.Axis X, B=@n36.Axis X) {materialFunction=&AppendMany}
  n38 = CameraVectorWS()
  n39 = SmoothStep(
   Min=DynamicParameter() {defaultValue=<rgba 0 0.75 0 1>, paramNames [Offset, Local Axis Fade Min, AlphaErosion, Param4]}.Local Axis Fade Min
   Value=Abs(None=DotProduct(A=@n37.RGB, B=@n38))
  )
  n40 = Constant() {r=1}
  n41 = SmoothStep(
   Max=OneMinus(None=ScalarParameter() {defaultValue=0.9, parameterName=Edge Hardness, sliderMax=1})
   Value=Multiply(A=Multiply(A=@n34, B=@n35.Result), B=StaticSwitchParameter(True=@n39, False=@n40) {parameterName=Use local Axis Fade})
  ) {constMax=0.1}
  n42 = DepthFade(Opacity=@n41, FadeDistance=ScalarParameter() {defaultValue=35, parameterName=Depth Fade Distance, sliderMax=100}) {fadeDistanceDefault=25}
  ShadingPathSwitch(Default=@n42, Deferred=@n42, Forward=@n41, Mobile=@n41)
 ]
}
