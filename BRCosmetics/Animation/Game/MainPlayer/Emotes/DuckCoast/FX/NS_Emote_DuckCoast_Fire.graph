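# Niagara system for the fire effect of the DuckCoast emote, built from delayed, short-lived ribbon emitters FireStart and Fire.
# FireStart shoots straight up, while Fire's velocity is scaled by three User FireVel parameters, with Y ramped by a curve.
# Curves over particle and loop age drive alpha fade and dynamic material parameters, and time of day scales RGB.
# Older mesh-flame emitters are disabled; glow, lick and smoke emitters list no renderers.
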NiagaraSystem /BRCosmetics/Animation/Game/MainPlayer/Emotes/DuckCoast/FX/NS_Emote_DuckCoast_Fire
dependencies {
 usedDataInterfaces [&NiagaraDataInterfaceArrayFloat, &NiagaraDataInterfaceCurve, &NiagaraDataInterfaceArrayFloat3]
 usedDynamicInputs [&FNTimeOfDayColor, &RandomRangeLinearColor]
 usedModules [
  &SystemState
  &self:SystemSpawnScript.SystemScriptSource.SystemScriptGraph.NiagaraNodeAssignment_0.SetVariables_#1
 ]
}
emitters {
 Fire {
  emitterUpdateScript [
   &EmitterState(Inactive Response=::NewEnumerator1, Life Cycle Mode=::NewEnumerator1, Loop Behavior=::NewEnumerator1, Loop Delay=0.75, UseLoopDelay=true)
   &SpawnRate(SpawnRate=60, Use Spawn Probability=false)
  ]
  particleSpawnScript [
   &InitializeParticle(
    Color Mode=::NewEnumerator0
    Lifetime=0.2
    Position Offset=&MakeVector(
     X=1
     Y=0
     Z=&FloatFromCurve(
      FloatCurve=DataInterface() {propertyValues {Curve {keys [
          <at 0 1 RCIM_Cubic>
          {arriveTangent=-0.1111, interpMode=RCIM_Cubic, leaveTangent=-0.1111, tangentMode=RCTM_User, time=0.7, value=1}
          <at 0.75 2 RCIM_Cubic>
         ]}}}
      CurveIndex=$Emitter.NormalizedLoopAge
      Scale Curve=3
     )
    )
    Ribbon Width=8.5
    Ribbon Width Mode=::NewEnumerator1
    UsePositionOffset=true
    Write Material Random=true
   )
   &AddVelocity(Velocity=<1 1 120>)
  ]
  particleUpdateScript [
   &ParticleState()
   &ScaleVelocity(Velocity Scale=&MakeVector(
     X=$User.Emote_DuckCoast_FireVelX
     Y=&Multiply_Float(
      A=&FloatFromCurve(
       FloatCurve=DataInterface() {propertyValues {Curve {keys [{arriveTangent=2.5688, interpMode=RCIM_Cubic, leaveTangent=2.5688, tangentMode=RCTM_User, time=0, value=0}, <at 1 1 RCIM_Cubic>]}}}
       Scale Curve=1.65
      )
      B=$User.Emote_DuckCoast_FireVelY
     )
     Z=$User.Emote_DuckCoast_FireVelZ
    ))
   &SolveForcesAndVelocity(Clamp Velocity=false, Limit Acceleration=false)
   &ScaleColor(
    Scale Alpha=&Multiply_Float(
     A=&FloatFromCurve(FloatCurve=DataInterface() {propertyValues {Curve {keys [
          <at 0 1>
          {arriveTangent=-0.1873, interpMode=RCIM_Cubic, leaveTangent=-0.1873, tangentMode=RCTM_User, time=0.5, value=1}
          {arriveTangent=-5.2851, interpMode=RCIM_Cubic, leaveTangent=-5.2851, tangentMode=RCTM_User, time=1, value=0}
         ]}}})
     B=&FloatFromCurve(
      FloatCurve=DataInterface() {propertyValues {Curve {keys [<at 0 1>, {arriveTangent=-5.2851, interpMode=RCIM_Cubic, leaveTangent=-5.2851, tangentMode=RCTM_User, time=1, value=0}]}}}
      CurveIndex=$Emitter.NormalizedLoopAge
     )
    )
    Scale RGB=&VectorFromFloat(Value=&FNTimeOfDayScalar(Night Scale=0.5, Morning Scale=0.75, Evening Scale=0.75))
    ScaleA=true
    ScaleRGB=true
   )
   &DynamicMaterialParameters(
    Index 0 Param 3=&FloatFromCurve(FloatCurve=DataInterface() {propertyValues {Curve {keys [<at 0 0 RCIM_Cubic>, {arriveTangent=3.2503, interpMode=RCIM_Cubic, leaveTangent=3.2503, tangentMode=RCTM_User, time=1, value=1}]}}})
    Module.Index0Float0Write=true
    Module.Index0Float1Write=true
    Module.Index0Float2Write=true
    Module.Index0Float3Write=true
   )
  ]
  rendererClasses [&NiagaraRibbonRendererProperties]
  renderers [{rendererClass=&NiagaraRibbonRendererProperties}]
 }
 FireGlow {rendererClasses [], renderers []}
 FireGlowStart {rendererClasses [], renderers []}
 FireOLD {
  bEnabled=false
  emitterUpdateScript [
   &EmitterState(Inactive Response=::NewEnumerator1, Life Cycle Mode=::NewEnumerator1, Loop Behavior=::NewEnumerator1, Loop Delay=0.85, UseLoopDelay=true)
   &SpawnBurst_Instantaneous(Spawn Count=2, Use Spawn Probability=false, UseLoopCountLimit=false)
   /SetVariables(
    Emitter.WPO Direction:Vector3f=&MakeVector(
     X=&FloatFromCurve(
      FloatCurve=DataInterface() {propertyValues {Curve {keys [<at 0.25 -1 RCIM_Cubic>, {arriveTangent=0.2666, interpMode=RCIM_Cubic, leaveTangent=0.2666, tangentMode=RCTM_User, time=1, value=1}]}}}
      CurveIndex=$Emitter.NormalizedLoopAge
     )
     Y=0
     Z=0
    )
    Emitter.WPO Intensity:NiagaraFloat=&FloatFromCurve(
     FloatCurve=DataInterface() {propertyValues {Curve {keys [
         <at 0 0>
         {arriveTangent=0.0784, interpMode=RCIM_Cubic, leaveTangent=0.0784, tangentMode=RCTM_User, time=0.7, value=0}
         <at 0.8 1 RCIM_Cubic>
        ]}}}
     CurveIndex=$Emitter.NormalizedLoopAge
     Scale Curve=5
    )
   ) {isSetParameters=true}
  ]
  particleSpawnScript [
   &PartitionParticles()
   &InitializeParticle(
    Lifetime=4.004
    Mesh Scale=&Multiply_VectorByFloat(Vector=<0.65 1.8 1>, Float=0.3)
    Mesh Scale Mode=::NewEnumerator4
    Position Offset=<1 0 2>
    UsePositionOffset=true
    Write Material Random=true
   )
   &InitialMeshOrientation(
    Mesh Axis To Orient=<0 0 1>
    Mesh Orientation Mode=::NewEnumerator2
    Rotation=&MakeVector(
     X=0
     Y=&SelectFloatFromArray(Direct Array Index=$Particles.PartitionParticles.Partition, Float Selection Array=DataInterface() {propertyValues {FloatData [0,0.25]}})
     Z=0
    )
    Use Rotation Vector=true
   ) {module=InitialMeshOrientation001}
   &InitialMeshOrientation(Mesh Orientation Mode=::NewEnumerator1, Rotation=<-0.25 0 0>, Use Rotation Vector=true)
  ]
  particleUpdateScript [
   &ParticleState()
   &ScaleMeshSize(Scale Factor=&MakeVector(
     X=0.675
     Y=&FloatFromCurve(
      FloatCurve=DataInterface() {propertyValues {Curve {keys [<at 0 1>, {arriveTangent=0.2, interpMode=RCIM_Cubic, leaveTangent=0.2, time=0.8, value=1}, <at 1 1.2 RCIM_Cubic>]}}}
      Scale Curve=0.2
     )
     Z=1
    ))
   /SetVariables(Particles.MeshOrientation=&MultiplyQuaternion(
     Quaternion A=$Particles.Initial.MeshOrientation
     Quaternion B=&MakeQuatFromAxisAngle(
      Axis=&SelectVectorFromArray(
       Direct Array Index=$Particles.PartitionParticles.Index In Partition
       Vector Selection Array=DataInterface() {propertyValues {FloatData [<0 0 1>, <-1 0 0>]}}
      )
      AngleInDegrees=&Lerp_Float(A=0, B=20, Alpha=$Particles.InterpolateOverTime002.InterpolateOverTime_Float.Moving Average)
     )
    )) {enabled=false, isSetParameters=true, module=SetVariables_#2}
   &DynamicMaterialParameters(
    Index 0 Param 1=&SelectFloatFromArray(Direct Array Index=$Particles.PartitionParticles.Partition, Float Selection Array=DataInterface() {propertyValues {FloatData [0,0.5]}})
    Module.Param0WriteEnabled=true
    Module.Param1WriteEnabled=false
    Module.Param2WriteEnabled=false
    Module.Param3WriteEnabled=false
   )
   &InterpolateOverTime(Rate Of Change=20, Target Vector Value:NiagaraPosition=$Particles.Position, Variable Type=::NewEnumerator2) {enabled=false}
   &DynamicMaterialParameters(
    Module.Param0WriteEnabled1=true
    Module.Param1WriteEnabled1=true
    Module.Param2WriteEnabled1=true
    Module.Param3WriteEnabled1=true
    Write Parameter Index 0=false
    Write Parameter Index 1=true
   ) {enabled=false, module=DynamicMaterialParameters001}
   &ScaleColor(
    Scale Alpha=&FloatFromCurve(FloatCurve=DataInterface() {propertyValues {Curve {keys [
         {arriveTangent=18.5021, interpMode=RCIM_Cubic, leaveTangent=18.5021, leaveTangentWeight=0.9117, tangentMode=RCTM_User, time=0, value=0}
         {
          arriveTangent=0.1183
          arriveTangentWeight=0.1678
          interpMode=RCIM_Cubic
          leaveTangent=0.1183
          leaveTangentWeight=0.1678
          tangentMode=RCTM_User
          time=0.1
          value=1
         }
         <at 1 1>
        ]}}})
    Scale RGB=&VectorFromFloat(Value=&FNTimeOfDayScalar(Night Scale=0.5, Morning Scale=0.75, Evening Scale=0.75))
    ScaleA=true
    ScaleRGB=true
   )
  ]
  rendererClasses [&NiagaraMeshRendererProperties]
  renderers [{rendererClass=&NiagaraMeshRendererProperties, rendererIndex=0}]
 }
 FireStart {
  emitterUpdateScript [
   &EmitterState(
    Inactive Response=::NewEnumerator1
    Life Cycle Mode=::NewEnumerator1
    Loop Behavior=::NewEnumerator1
    Loop Delay=0.75
    Loop Duration=0.727
    UseLoopDelay=true
   )
   &SpawnRate(SpawnRate=60, Use Spawn Probability=false)
  ]
  particleSpawnScript [
   &InitializeParticle(
    Color Mode=::NewEnumerator0
    Lifetime=0.2
    Position Offset=<1 0 0>
    Ribbon Width=10
    Ribbon Width Mode=::NewEnumerator1
    UsePositionOffset=true
    Write Material Random=true
   )
   &AddVelocity(Velocity=<0 0 727>)
  ]
  particleUpdateScript [
   &ParticleState()
   &SolveForcesAndVelocity(Clamp Velocity=false, Limit Acceleration=false)
   &ScaleColor(
    Scale Alpha=&Multiply_Float(
     A=&FloatFromCurve(FloatCurve=DataInterface() {propertyValues {Curve {keys [
          <at 0 1>
          {arriveTangent=-0.1873, interpMode=RCIM_Cubic, leaveTangent=-0.1873, tangentMode=RCTM_User, time=0.5, value=1}
          {arriveTangent=-5.2851, interpMode=RCIM_Cubic, leaveTangent=-5.2851, tangentMode=RCTM_User, time=1, value=0}
         ]}}})
     B=&FloatFromCurve(
      FloatCurve=DataInterface() {propertyValues {Curve {keys [<at 0 1>, {arriveTangent=-5.2851, interpMode=RCIM_Cubic, leaveTangent=-5.2851, tangentMode=RCTM_User, time=1, value=0}]}}}
      CurveIndex=$Emitter.NormalizedLoopAge
     )
    )
    Scale RGB=&VectorFromFloat(Value=&FNTimeOfDayScalar(Night Scale=0.5, Morning Scale=0.75, Evening Scale=0.75))
    ScaleA=true
    ScaleRGB=true
   )
   &DynamicMaterialParameters(
    Index 0 Param 3=&FloatFromCurve(
     FloatCurve=DataInterface() {propertyValues {Curve {keys [<at 0 0>, {arriveTangent=1, interpMode=RCIM_Cubic, leaveTangent=1, time=0.25, value=0}, <at 1 1 RCIM_Cubic>]}}}
     CurveIndex=$Emitter.NormalizedLoopAge
    )
    Module.Index0Float0Write=true
    Module.Index0Float1Write=true
    Module.Index0Float2Write=true
    Module.Index0Float3Write=true
   )
  ]
  rendererClasses [&NiagaraRibbonRendererProperties]
  renderers [{rendererClass=&NiagaraRibbonRendererProperties}]
 }
 FireStartOLD {
  bEnabled=false
  emitterUpdateScript [
   &EmitterState(Inactive Response=::NewEnumerator1, Life Cycle Mode=::NewEnumerator1, Loop Behavior=::NewEnumerator1, Loop Delay=0.85, UseLoopDelay=true)
   &SpawnBurst_Instantaneous(Spawn Count=2, Use Spawn Probability=false, UseLoopCountLimit=false)
  ]
  particleSpawnScript [
   &PartitionParticles()
   &InitializeParticle(
    Lifetime=0.65
    Mesh Scale=&Multiply_VectorByFloat(Vector=<0.65 1.8 1>, Float=0.3)
    Mesh Scale Mode=::NewEnumerator4
    Position Offset=<0 0 -8>
    UsePositionOffset=false
    Write Material Random=true
   )
   &InitialMeshOrientation(
    Mesh Axis To Orient=<0 0 1>
    Mesh Orientation Mode=::NewEnumerator2
    Rotation=&MakeVector(
     X=0
     Y=&SelectFloatFromArray(Direct Array Index=$Particles.PartitionParticles.Partition, Float Selection Array=DataInterface() {propertyValues {FloatData [0,0.25]}})
     Z=0
    )
    Use Rotation Vector=true
   ) {module=InitialMeshOrientation001}
   &InitialMeshOrientation(Mesh Orientation Mode=::NewEnumerator1, Rotation=<-0.25 0 0>, Use Rotation Vector=true)
  ]
  particleUpdateScript [
   &ParticleState()
   &ScaleMeshSize(Scale Factor=&MakeVector(
     X=&FloatFromCurve(FloatCurve=DataInterface() {propertyValues {Curve {
         defaultValue=0
         keys [
          {interpMode=RCIM_Cubic, tangentMode=RCTM_User, time=0, value=0}
          {interpMode=RCIM_Cubic, tangentMode=RCTM_User, time=0.25, value=0.7}
          {tangentMode=RCTM_User, time=1, value=1.2}
         ]
        }}})
     Y=&FloatFromCurve(FloatCurve=DataInterface() {propertyValues {Curve {keys [
          {arriveTangent=3.9222, interpMode=RCIM_Cubic, leaveTangent=3.9222, tangentMode=RCTM_User, time=0, value=0}
          {arriveTangent=-0.0274, interpMode=RCIM_Cubic, leaveTangent=-0.0274, tangentMode=RCTM_User, time=0.5, value=0.8}
         ]}}})
     Z=1
    ))
   /SetVariables(Particles.MeshOrientation=&MultiplyQuaternion(
     Quaternion A=$Particles.Initial.MeshOrientation
     Quaternion B=&MakeQuatFromAxisAngle(
      Axis=&SelectVectorFromArray(
       Direct Array Index=$Particles.PartitionParticles.Index In Partition
       Vector Selection Array=DataInterface() {propertyValues {FloatData [<0 0 1>, <-1 0 0>]}}
      )
      AngleInDegrees=&Lerp_Float(A=0, B=20, Alpha=$Particles.InterpolateOverTime002.InterpolateOverTime_Float.Moving Average)
     )
    )) {enabled=false, isSetParameters=true, module=SetVariables_#2}
   &DynamicMaterialParameters(
    Index 0 Param 1=&SelectFloatFromArray(Direct Array Index=$Particles.PartitionParticles.Partition, Float Selection Array=DataInterface() {propertyValues {FloatData [0,0.5]}})
    Index 2 Param 1=&FloatFromCurve(FloatCurve=DataInterface() {propertyValues {Curve {keys [<at 0 1 RCIM_Cubic>, {arriveTangent=-1.8037, interpMode=RCIM_Cubic, leaveTangent=-1.8037, tangentMode=RCTM_User, time=1, value=0}]}}})
    Index 2 Param 2=&FloatFromCurve(FloatCurve=DataInterface() {propertyValues {Curve {
        defaultValue=0
        keys [
         {arriveTangent=0.1253, interpMode=RCIM_Cubic, leaveTangent=0.1253, tangentMode=RCTM_User, time=0.25, value=0}
         {arriveTangent=0.0224, interpMode=RCIM_Cubic, leaveTangent=0.0224, tangentMode=RCTM_User, time=1, value=1}
        ]
       }}})
    Module.Param0WriteEnabled=true
    Module.Param0WriteEnabled2=true
    Module.Param1WriteEnabled=false
    Module.Param1WriteEnabled2=true
    Module.Param2WriteEnabled=false
    Module.Param2WriteEnabled2=true
    Module.Param3WriteEnabled=false
    Module.Param3WriteEnabled2=true
    Write Parameter Index 2=true
   )
   &InterpolateOverTime(Rate Of Change=20, Target Vector Value:NiagaraPosition=$Particles.Position, Variable Type=::NewEnumerator2) {enabled=false}
   &DynamicMaterialParameters(
    Module.Param0WriteEnabled1=true
    Module.Param1WriteEnabled1=true
    Module.Param2WriteEnabled1=true
    Module.Param3WriteEnabled1=true
    Write Parameter Index 0=false
    Write Parameter Index 1=true
   ) {enabled=false, module=DynamicMaterialParameters001}
   &ScaleColor(
    Scale Alpha=&FloatFromCurve(FloatCurve=DataInterface() {propertyValues {Curve {keys [
         <at 0 1>
         {arriveTangent=-0.1183, interpMode=RCIM_Cubic, leaveTangent=-0.1183, tangentMode=RCTM_User, time=0.5, value=1}
         {arriveTangent=-5.3778, interpMode=RCIM_Cubic, leaveTangent=-5.3778, tangentMode=RCTM_User, time=1, value=0}
        ]}}})
    Scale RGB=&VectorFromFloat(Value=&FNTimeOfDayScalar(Night Scale=0.5, Morning Scale=0.75, Evening Scale=0.75))
    ScaleA=true
    ScaleRGB=true
   )
   /SetVariables(Particles.Position=&AddVectorToPosition(
     Position=/Niagara/DynamicInputs/Helpers/SimulationPosition()
     Vector=&MakeVector(
      X=1
      Y=0
      Z=&FloatFromCurve(FloatCurve=DataInterface() {propertyValues {Curve {keys [<at 0.8 0 RCIM_Cubic>,<at 1 5 RCIM_Cubic>]}}})
     )
    )) {isSetParameters=true}
  ]
  rendererClasses [&NiagaraMeshRendererProperties]
  renderers [{rendererClass=&NiagaraMeshRendererProperties, rendererIndex=0}]
 }
 Fire_Licks {rendererClasses [], renderers []}
 GlowCone {rendererClasses [], renderers []}
 GlowCone001 {rendererClasses [], renderers []}
 Smoke {rendererClasses [], renderers []}
}
summary {
 systemName=NS_Emote_DuckCoast_Fire
 emitters [
  {emitterName=GlowCone, rendererClasses []}
  {emitterName=FireGlowStart, rendererClasses []}
  {emitterName=FireGlow, rendererClasses []}
  {bEnabled=false, emitterName=FireStartOLD, rendererClasses [&NiagaraMeshRendererProperties]}
  {bEnabled=false, emitterName=FireOLD, rendererClasses [&NiagaraMeshRendererProperties]}
  {emitterName=GlowCone001, rendererClasses []}
  {emitterName=FireStart, rendererClasses [&NiagaraRibbonRendererProperties]}
  {emitterName=Fire, rendererClasses [&NiagaraRibbonRendererProperties]}
  {emitterName=Fire_Licks, rendererClasses []}
  {emitterName=Smoke, rendererClasses []}
 ]
 userVariables [
  User.Emote_DuckCoast_FireVelX {type=NiagaraFloat, defaultValue=0, description=""}
  User.Emote_DuckCoast_FireVelY {type=NiagaraFloat, defaultValue=0, description=""}
  User.Emote_DuckCoast_FireVelZ {type=NiagaraFloat, defaultValue=0, description=""}
 ]
}
