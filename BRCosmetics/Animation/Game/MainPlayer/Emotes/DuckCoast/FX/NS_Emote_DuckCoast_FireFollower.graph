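# Niagara system for the DuckCoast player emote, drawing short-lived ribbon trails of fire and smoke.
# Fire spawns upward-moving particles whose sideways Y velocity is scaled by the FireVelY user parameter, with loop-age-driven offset and fade and time-of-day color dimming.
# Smoke starts after a loop delay, shaping its narrowing ribbon by setting positions along curves relative to the emitter.
# The older mesh-based fire emitter is disabled, and FireGlow shows no renderer.
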NiagaraSystem /BRCosmetics/Animation/Game/MainPlayer/Emotes/DuckCoast/FX/NS_Emote_DuckCoast_FireFollower
dependencies {
 usedDataInterfaces [&NiagaraDataInterfaceCurve, &NiagaraDataInterfaceArrayFloat, &NiagaraDataInterfaceArrayFloat3]
 usedDynamicInputs []
 usedModules [&SystemState]
}
emitters {
 Fire {
  emitterUpdateScript [
   &EmitterState(
    Inactive Response=::NewEnumerator1
    Life Cycle Mode=::NewEnumerator1
    Loop Behavior=::NewEnumerator1
    Loop Delay=0.85
    Loop Duration=1.333
    UseLoopDelay=false
   )
   &SpawnRate(SpawnRate=60, Use Spawn Probability=false)
  ]
  particleSpawnScript [
   &InitializeParticle(
    Color Mode=::NewEnumerator0
    Lifetime=0.2
    Position Offset=&MakeVector(
     X=1
     Y=0
     Z=&FloatFromCurve(
      FloatCurve=DataInterface() {propertyValues {Curve {keys [
          <at 0 1 RCIM_Cubic>
          {arriveTangent=-0.1111, interpMode=RCIM_Cubic, leaveTangent=-0.1111, tangentMode=RCTM_User, time=0.7, value=1}
          <at 0.75 2 RCIM_Cubic>
         ]}}}
      CurveIndex=$Emitter.NormalizedLoopAge
      Scale Curve=3
     )
    )
    Ribbon Width=8.5
    Ribbon Width Mode=::NewEnumerator1
    UsePositionOffset=true
    Write Material Random=true
   )
   &AddVelocity(Velocity=<1 1 120>)
  ]
  particleUpdateScript [
   &ParticleState()
   &ScaleVelocity(Velocity Scale=&MakeVector(
     X=0
     Y=&Multiply_Float(
      A=&FloatFromCurve(
       FloatCurve=DataInterface() {propertyValues {Curve {keys [{arriveTangent=2.5688, interpMode=RCIM_Cubic, leaveTangent=2.5688, tangentMode=RCTM_User, time=0, value=0}, <at 1 1 RCIM_Cubic>]}}}
       Scale Curve=1.65
      )
      B=$User.Emote_DuckCoast_FireVelY
     )
     Z=1
    ))
   &SolveForcesAndVelocity(Clamp Velocity=false, Limit Acceleration=false)
   &ScaleColor(
    Scale Alpha=&Multiply_Float(
     A=&FloatFromCurve(FloatCurve=DataInterface() {propertyValues {Curve {keys [
          <at 0 1>
          {arriveTangent=-0.1873, interpMode=RCIM_Cubic, leaveTangent=-0.1873, tangentMode=RCTM_User, time=0.5, value=1}
          {arriveTangent=-5.2851, interpMode=RCIM_Cubic, leaveTangent=-5.2851, tangentMode=RCTM_User, time=1, value=0}
         ]}}})
     B=&FloatFromCurve(
      FloatCurve=DataInterface() {propertyValues {Curve {keys [<at 0 1>, {arriveTangent=-5.2851, interpMode=RCIM_Cubic, leaveTangent=-5.2851, tangentMode=RCTM_User, time=1, value=0}]}}}
      CurveIndex=$Emitter.NormalizedLoopAge
     )
    )
    Scale RGB=&VectorFromFloat(Value=&FNTimeOfDayScalar(Night Scale=0.5, Morning Scale=0.75, Evening Scale=0.75))
    ScaleA=true
    ScaleRGB=true
   )
   &DynamicMaterialParameters(
    Index 0 Param 3=&FloatFromCurve(FloatCurve=DataInterface() {propertyValues {Curve {keys [<at 0 0 RCIM_Cubic>, {arriveTangent=3.2503, interpMode=RCIM_Cubic, leaveTangent=3.2503, tangentMode=RCTM_User, time=1, value=1}]}}})
    Module.Index0Float0Write=true
    Module.Index0Float1Write=true
    Module.Index0Float2Write=true
    Module.Index0Float3Write=true
   )
  ]
  rendererClasses [&NiagaraRibbonRendererProperties]
  renderers [{rendererClass=&NiagaraRibbonRendererProperties}]
 }
 FireGlow {rendererClasses [], renderers []}
 FireOLD {
  bEnabled=false
  emitterUpdateScript [
   &EmitterState(
    Inactive Response=::NewEnumerator1
    Life Cycle Mode=::NewEnumerator1
    Loop Behavior=::NewEnumerator1
    Loop Delay=0.85
    Loop Duration=1.333
    UseLoopDelay=false
   )
   &SpawnBurst_Instantaneous(Spawn Count=2, Use Spawn Probability=false, UseLoopCountLimit=false)
   /SetVariables(
    Emitter.WPO Direction:Vector3f=<0 -1 0>
    Emitter.WPO Intensity:NiagaraFloat=&FloatFromCurve(
     FloatCurve=DataInterface() {propertyValues {Curve {keys [<at 0 0>, {arriveTangent=0.0784, interpMode=RCIM_Cubic, leaveTangent=0.0784, tangentMode=RCTM_User, time=0.9, value=0}, <at 1 1 RCIM_Cubic>]}}}
     CurveIndex=$Emitter.NormalizedLoopAge
     Scale Curve=2
    )
   ) {isSetParameters=true}
  ]
  particleSpawnScript [
   &PartitionParticles()
   &InitializeParticle(
    Lifetime=1.333
    Mesh Scale=&Multiply_VectorByFloat(Vector=<0.65 1.8 1>, Float=0.3)
    Mesh Scale Mode=::NewEnumerator4
    Position Offset=<1 0 2>
    UsePositionOffset=true
    Write Material Random=true
   )
   &InitialMeshOrientation(
    Mesh Axis To Orient=<0 0 1>
    Mesh Orientation Mode=::NewEnumerator2
    Rotation=&MakeVector(
     X=0
     Y=&SelectFloatFromArray(Direct Array Index=$Particles.PartitionParticles.Partition, Float Selection Array=DataInterface() {propertyValues {FloatData [0,0.25]}})
     Z=0
    )
    Use Rotation Vector=true
   ) {module=InitialMeshOrientation001}
   &InitialMeshOrientation(Mesh Orientation Mode=::NewEnumerator1, Rotation=<-0.25 0 0>, Use Rotation Vector=true)
  ]
  particleUpdateScript [
   &ParticleState()
   &ScaleMeshSize(Scale Factor=&MakeVector(
     X=0.6
     Y=&FloatFromCurve(
      FloatCurve=DataInterface() {propertyValues {Curve {keys [<at 0 1>, {arriveTangent=-0.2, interpMode=RCIM_Cubic, leaveTangent=-0.2, time=0.8, value=1}, <at 1 0.8 RCIM_Cubic>]}}}
      Scale Curve=0.2
     )
     Z=1
    ))
   /SetVariables(Particles.MeshOrientation=&MultiplyQuaternion(
     Quaternion A=$Particles.Initial.MeshOrientation
     Quaternion B=&MakeQuatFromAxisAngle(
      Axis=&SelectVectorFromArray(
       Direct Array Index=$Particles.PartitionParticles.Index In Partition
       Vector Selection Array=DataInterface() {propertyValues {FloatData [<0 0 1>, <-1 0 0>]}}
      )
      AngleInDegrees=&Lerp_Float(A=0, B=20, Alpha=$Particles.InterpolateOverTime002.InterpolateOverTime_Float.Moving Average)
     )
    )) {enabled=false, isSetParameters=true}
   &DynamicMaterialParameters(
    Index 0 Param 1=&SelectFloatFromArray(Direct Array Index=$Particles.PartitionParticles.Partition, Float Selection Array=DataInterface() {propertyValues {FloatData [0,0.5]}})
    Module.Param0WriteEnabled=true
    Module.Param1WriteEnabled=false
    Module.Param2WriteEnabled=false
    Module.Param3WriteEnabled=false
   )
   &InterpolateOverTime(Rate Of Change=20, Target Vector Value:NiagaraPosition=$Particles.Position, Variable Type=::NewEnumerator2) {enabled=false}
   &DynamicMaterialParameters(
    Module.Param0WriteEnabled1=true
    Module.Param1WriteEnabled1=true
    Module.Param2WriteEnabled1=true
    Module.Param3WriteEnabled1=true
    Write Parameter Index 0=false
    Write Parameter Index 1=true
   ) {enabled=false, module=DynamicMaterialParameters001}
   &ScaleColor(
    Scale Alpha=&FloatFromCurve(FloatCurve=DataInterface() {propertyValues {Curve {keys [
         {arriveTangent=18.5021, interpMode=RCIM_Cubic, leaveTangent=18.5021, leaveTangentWeight=0.9117, tangentMode=RCTM_User, time=0, value=0}
         {
          arriveTangent=0.1183
          arriveTangentWeight=0.1678
          interpMode=RCIM_Cubic
          leaveTangent=0.1183
          leaveTangentWeight=0.1678
          tangentMode=RCTM_User
          time=0.1
          value=1
         }
         <at 1 1>
        ]}}})
    Scale RGB=&VectorFromFloat(Value=&FNTimeOfDayScalar(Night Scale=0.5, Morning Scale=0.75, Evening Scale=0.75))
    ScaleA=true
    ScaleRGB=true
   )
  ]
  rendererClasses [&NiagaraMeshRendererProperties]
  renderers [{rendererClass=&NiagaraMeshRendererProperties, rendererIndex=0}]
 }
 Smoke {
  emitterUpdateScript [
   &EmitterState(Life Cycle Mode=::NewEnumerator1, Loop Behavior=::NewEnumerator1, Loop Delay=1.275, Loop Duration=0.925, UseLoopDelay=true)
   &SpawnRate(
    SpawnRate=&FloatFromCurve(
     FloatCurve=DataInterface() {propertyValues {Curve {keys [<at 0 1>,<at 0.8 1>,<at 1 0>]}}}
     CurveIndex=$Emitter.NormalizedLoopAge
     Scale Curve=120
    )
    Use Spawn Probability=false
   )
  ]
  particleSpawnScript [&InitializeParticle(Lifetime=0.2, Ribbon Width=12, Ribbon Width Mode=::NewEnumerator1, UsePositionOffset=false, Write Material Random=true)]
  particleUpdateScript [
   &ParticleState()
   &ScaleRibbonWidth(Ribbon Width Scale=&FloatFromCurve(FloatCurve=DataInterface() {propertyValues {Curve {
        defaultValue=0
        keys [<at 0 1 RCIM_Cubic>, {arriveTangent=-4.3709, interpMode=RCIM_Cubic, leaveTangent=-4.3709, tangentMode=RCTM_User, time=1, value=0}]
       }}}))
   &DynamicMaterialParameters(
    Index 0 Param 1=&FloatFromCurve(
     FloatCurve=DataInterface() {propertyValues {Curve {keys [{arriveTangent=-1.8184, interpMode=RCIM_Cubic, leaveTangent=-1.8184, tangentMode=RCTM_User, time=0, value=1}, <at 1 0 RCIM_Cubic>]}}}
     CurveIndex=$Emitter.NormalizedLoopAge
    )
    Index 0 Param 3=1
    Index 0 Param 4=0.125
    Module.Index0Float0Write=true
    Module.Index0Float1Write=true
    Module.Index0Float2Write=true
    Module.Index0Float3Write=true
   )
   /SetVariables(Particles.Position=&AddVectorToPosition(
     Position=$Engine.Emitter.SimulationPosition
     Vector=&MakeVector(
      X=0
      Y=&FloatFromCurve(FloatCurve=DataInterface() {propertyValues {Curve {keys [<at 0 0>,<at 1 25>]}}})
      Z=&FloatFromCurve(FloatCurve=DataInterface() {propertyValues {Curve {
          defaultValue=0
          keys [
           {
            arriveTangent=44.3979
            arriveTangentWeight=0.2925
            interpMode=RCIM_Cubic
            leaveTangent=44.3979
            leaveTangentWeight=0.2925
            tangentMode=RCTM_User
            time=0
            value=4.951
           }
           {
            arriveTangent=19.9391
            arriveTangentWeight=0.2653
            interpMode=RCIM_Cubic
            leaveTangent=19.9391
            leaveTangentWeight=0.2653
            tangentMode=RCTM_User
            time=0.2
            value=15
           }
           <at 1 15 RCIM_Cubic>
          ]
         }}})
     )
    )) {isSetParameters=true}
  ]
  rendererClasses [&NiagaraRibbonRendererProperties]
  renderers [{rendererClass=&NiagaraRibbonRendererProperties}]
 }
}
summary {
 systemName=NS_Emote_DuckCoast_FireFollower
 emitters [
  {emitterName=FireGlow, rendererClasses []}
  {bEnabled=false, emitterName=FireOLD, rendererClasses [&NiagaraMeshRendererProperties]}
  {emitterName=Smoke, rendererClasses [&NiagaraRibbonRendererProperties]}
  {emitterName=Fire, rendererClasses [&NiagaraRibbonRendererProperties]}
 ]
 userVariables [
  User.Emote_DuckCoast_FireVelX {type=NiagaraFloat, defaultValue=0, description=""}
  User.Emote_DuckCoast_FireVelY {type=NiagaraFloat, defaultValue=0, description=""}
 ]
}
