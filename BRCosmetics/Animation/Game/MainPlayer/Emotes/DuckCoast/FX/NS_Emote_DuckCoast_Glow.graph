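# Niagara glow effect for the DuckCoast emote, bursting bright HDR green sprites that follow a skeletal mesh.
# SkeletalMeshLocation runs in particle update through a system-level mesh reader, keeping particles attached to the sampled bones or sockets.
# An alpha curve quickly fades each short-lived sprite in and out.
# User.bGlow, a float, sets an emitter Glow bool by comparison with zero; bOther and bSlurp are declared without visible use.
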NiagaraSystem /BRCosmetics/Animation/Game/MainPlayer/Emotes/DuckCoast/FX/NS_Emote_DuckCoast_Glow
dependencies {
 usedDataInterfaces [&NiagaraDataInterfaceCurve, &NiagaraDataInterfaceSkeletalMesh]
 usedDynamicInputs []
 usedModules [
  &self:SystemSpawnScript.SystemScriptSource.SystemScriptGraph.NiagaraNodeAssignment_0.SetVariables_#1
  &SystemState
 ]
}
emitters {Glow {
  emitterSpawnScript [/SetVariables(Emitter.Glow:NiagaraBool=&SetBoolByFloatComparison(A=$User.bGlow, B=0)) {isSetParameters=true}]
  emitterUpdateScript [
   &EmitterState()
   &SpawnBurst_Instantaneous(Use Spawn Probability=false, UseLoopCountLimit=false)
  ]
  particleSpawnScript [&InitializeParticle(
    Color=<rgba 2.4489 12 1.9641 0.05>
    Lifetime=2.23
    Sprite Size=<30 40>
    Sprite Size Mode=::NewEnumerator4
    UsePositionOffset=false
    Write Material Random=true
   )]
  particleUpdateScript [
   &ParticleState()
   &SkeletalMeshLocation(Bone Sampling Mode=::NewEnumerator4, MaskSpawnGroup=false, Skeletal Mesh=$System.Skeletal Mesh Reader)
   &ScaleColor(
    Scale Alpha=&FloatFromCurve(FloatCurve=DataInterface() {propertyValues {Curve {keys [
         <at 0 0 RCIM_Cubic>
         {arriveTangent=-0.0999, interpMode=RCIM_Cubic, leaveTangent=-0.0999, tangentMode=RCTM_User, time=0.05, value=1}
         {arriveTangent=-0.1889, interpMode=RCIM_Cubic, leaveTangent=-0.1889, tangentMode=RCTM_User, time=0.9, value=1}
         <at 1 0 RCIM_Cubic>
        ]}}})
    ScaleA=true
    ScaleRGB=true
   )
  ]
  rendererClasses [&NiagaraSpriteRendererProperties]
  renderers [{rendererClass=&NiagaraSpriteRendererProperties}]
 }}
summary {
 systemName=NS_Emote_DuckCoast_Glow
 emitters [{emitterName=Glow, rendererClasses [&NiagaraSpriteRendererProperties]}]
 userVariables [
  User.bGlow {type=NiagaraFloat, defaultValue=0, description=""}
  User.bOther {type=NiagaraFloat, defaultValue=0, description=""}
  User.bSlurp {type=NiagaraFloat, defaultValue=0, description=""}
 ]
}
