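# Niagara system for the DuckCoast emote that draws two short-lived liquid streams, orange SlurpJuice and purple OtherJuice, as ribbons.
# Each emitter follows its own bone or socket through a shared skeletal mesh reader, with looped-age curves staggering when each stream spawns.
# Particle position is rewritten every update as the sampled point plus curved offsets, faking a falling arc without forces while ribbons widen and fade.
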NiagaraSystem /BRCosmetics/Animation/Game/MainPlayer/Emotes/DuckCoast/FX/NS_Emote_DuckCoast_Liquid
dependencies {
 usedDataInterfaces [&NiagaraDataInterfaceCurve, &NiagaraDataInterfaceColorCurve, &NiagaraDataInterfaceSkeletalMesh]
 usedDynamicInputs []
 usedModules [
  &self:SystemSpawnScript.SystemScriptSource.SystemScriptGraph.NiagaraNodeAssignment_0.SetVariables_#1
  &SystemState
 ]
}
emitters {
 OtherJuice {
  emitterUpdateScript [
   &EmitterState()
   &SpawnRate(
    SpawnRate=&FloatFromCurve(
     FloatCurve=DataInterface() {propertyValues {Curve {keys [
         <at 1.3 0>
         <at 1.34 1>
         {arriveTangent=-0.0118, interpMode=RCIM_Cubic, leaveTangent=-0.0118, tangentMode=RCTM_User, time=2, value=1}
         <at 2.4 0 RCIM_Cubic>
        ]}}}
     CurveIndex=$Emitter.LoopedAge
     Scale Curve=60
    )
    Use Spawn Probability=false
   )
  ]
  particleSpawnScript [&InitializeParticle(
    Color=&ColorFromCurve(
     ColorCurve=DataInterface() {propertyValues {
       AlphaCurve {keys [<at 0 0 RCIM_Cubic>,<at 0.157 0.8>,<at 1 0.8>]}
       BlueCurve {keys [<at 0 0.2744 RCIM_Cubic>]}
       GreenCurve {keys [<at 0 0 RCIM_Cubic>]}
       RedCurve {keys [<at 0 0.35 RCIM_Cubic>]}
      }}
     CurveIndex=$Emitter.NormalizedLoopAge
    )
    Lifetime=0.3
    Position Offset=<-1.2 0 -2>
    Ribbon Width=4
    Ribbon Width Mode=::NewEnumerator1
    UsePositionOffset=false
    Write Material Random=true
   )]
  particleUpdateScript [
   &SkeletalMeshLocation(
    Bone / Socket Index=1
    MaskSpawnGroup=false
    Mesh Sampling Type=::NewEnumerator5
    Position Sampling=::NewEnumerator2
    Sampled Position Offset=<0 0 -2>
    Skeletal Mesh=$System.Skeletal Mesh Reader
    Socket Sampling Mode=::NewEnumerator5
   )
   &InheritSourceMovement(ApplyForce=false, ApplyPosition=true, ApplyPositionFalloff=false) {enabled=false}
   &SolveForcesAndVelocity(Clamp Velocity=false, Limit Acceleration=false)
   &ParticleState()
   &ScaleColor(
    Scale Alpha=&FloatFromCurve(FloatCurve=DataInterface() {propertyValues {Curve {keys [
         {interpMode=RCIM_Cubic, tangentMode=RCTM_User, time=0, value=1}
         {interpMode=RCIM_Cubic, tangentMode=RCTM_User, time=0.7, value=1}
         {tangentMode=RCTM_User, time=1, value=0}
        ]}}})
    ScaleA=true
    ScaleRGB=true
   )
   &ScaleRibbonWidth(Ribbon Width Scale=&FloatFromCurve(FloatCurve=DataInterface() {propertyValues {Curve {keys [<at 0 0.3 RCIM_Cubic>,<at 1 1 RCIM_Cubic>]}}}))
   /SetVariables(Particles.Position=&AddVectorToPosition(
     Position=$Particles.SkeletalMeshLocation.SampledPosition
     Vector=&MakeVector(
      X=&FloatFromCurve(FloatCurve=DataInterface() {propertyValues {Curve {keys [<at 0 0 RCIM_Cubic>,<at 0.2 2 RCIM_Cubic>]}}})
      Y=1.125
      Z=&Add_Float(
       A=1
       B=&Multiply_Float(
        A=&FloatFromCurve(
         FloatCurve=DataInterface() {propertyValues {Curve {keys [<at 0 0 RCIM_Cubic>, {arriveTangent=2.9167, interpMode=RCIM_Cubic, leaveTangent=2.9167, tangentMode=RCTM_User, time=1, value=25}]}}}
         Scale Curve=-1
        )
        B=&FloatFromCurve(
         FloatCurve=DataInterface() {propertyValues {Curve {keys [<at 0 1>,<at 0.7 1>,<at 1 2.5 RCIM_Cubic>]}}}
         CurveIndex=$Emitter.NormalizedLoopAge
        )
       )
      )
     )
    )) {isSetParameters=true}
  ]
  rendererClasses [&NiagaraRibbonRendererProperties]
  renderers [{rendererClass=&NiagaraRibbonRendererProperties}]
 }
 SlurpJuice {
  emitterUpdateScript [
   &EmitterState()
   &SpawnRate(
    SpawnRate=&FloatFromCurve(
     FloatCurve=DataInterface() {propertyValues {Curve {keys [<at 0 1>, {arriveTangent=-0.4545, interpMode=RCIM_Cubic, leaveTangent=-0.4545, time=1.8, value=1}, <at 2.2 0 RCIM_Cubic>]}}}
     CurveIndex=$Emitter.LoopedAge
     Scale Curve=60
    )
    Use Spawn Probability=false
   )
  ]
  particleSpawnScript [&InitializeParticle(
    Color=&ColorFromCurve(
     ColorCurve=DataInterface() {propertyValues {
       AlphaCurve {keys [<at 0 0 RCIM_Cubic>,<at 0.157 0.8>,<at 1 0.8>]}
       BlueCurve {keys [<at 0 0>]}
       GreenCurve {keys [<at 0 0.1612>]}
       RedCurve {keys [<at 0 1>]}
      }}
     CurveIndex=$Emitter.NormalizedLoopAge
    )
    Lifetime=0.2
    Position Offset=<-1.2 0 -2>
    Ribbon Width=4
    Ribbon Width Mode=::NewEnumerator1
    UsePositionOffset=false
    Write Material Random=true
   )]
  particleUpdateScript [
   &SkeletalMeshLocation(
    MaskSpawnGroup=false
    Mesh Sampling Type=::NewEnumerator5
    Position Sampling=::NewEnumerator2
    Skeletal Mesh=$System.Skeletal Mesh Reader
    Socket Sampling Mode=::NewEnumerator5
   )
   &InheritSourceMovement(ApplyForce=false, ApplyPosition=true, ApplyPositionFalloff=false) {enabled=false}
   &SolveForcesAndVelocity(Clamp Velocity=false, Limit Acceleration=false)
   &ParticleState()
   &ScaleColor(
    Scale Alpha=&FloatFromCurve(FloatCurve=DataInterface() {propertyValues {Curve {keys [
         {interpMode=RCIM_Cubic, tangentMode=RCTM_User, time=0, value=1}
         {interpMode=RCIM_Cubic, tangentMode=RCTM_User, time=0.7, value=1}
         {tangentMode=RCTM_User, time=1, value=0}
        ]}}})
    ScaleA=true
    ScaleRGB=true
   )
   &ScaleRibbonWidth(Ribbon Width Scale=&FloatFromCurve(FloatCurve=DataInterface() {propertyValues {Curve {keys [<at 0 0.3 RCIM_Cubic>,<at 1 1 RCIM_Cubic>]}}}))
   /SetVariables(Particles.Position=&AddVectorToPosition(
     Position=$Particles.SkeletalMeshLocation.SampledPosition
     Vector=&MakeVector(
      X=&FloatFromCurve(FloatCurve=DataInterface() {propertyValues {Curve {keys [<at 0 0 RCIM_Cubic>,<at 0.3 1.75 RCIM_Cubic>]}}}, Scale Curve=-1)
      Y=-0.25
      Z=&Add_Float(
       A=1
       B=&FloatFromCurve(FloatCurve=DataInterface() {propertyValues {Curve {keys [<at 0 0 RCIM_Cubic>,<at 1 25 RCIM_Cubic>]}}}, Scale Curve=-1)
      )
     )
    )) {isSetParameters=true}
  ]
  rendererClasses [&NiagaraRibbonRendererProperties]
  renderers [{rendererClass=&NiagaraRibbonRendererProperties}]
 }
}
summary {
 systemName=NS_Emote_DuckCoast_Liquid
 emitters [
  {emitterName=SlurpJuice, rendererClasses [&NiagaraRibbonRendererProperties]}
  {emitterName=OtherJuice, rendererClasses [&NiagaraRibbonRendererProperties]}
 ]
 userVariables [User.bGlow {type=NiagaraFloat, defaultValue=0, description=""}]
}
